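# Chapter 4 Season 5 Item Shop background material, blending dark tones, tinted smoke and red lasers into the MPItemShop_Setup function's color input.
# Smoke comes from a flow-mapped noise blended with two panning noise textures and shaped by linear gradients.
# Four parameter-positioned, rotated box SDFs are smoothstepped into thin, large and haze laser masks.
# Named reroutes carry screen-space scaled UVs and masks between stages, alongside a panning particle mask.
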Material /ItemShop/UI/Art/Ch4S4/M_UI_StoreBG_Ch4S5_MP
graph {
 properties {pixelDepthOffsetMode=PDOM_Legacy, usedShadingModels=Default Lit}
 nodes [
  Fmod()
  n1 = VectorParameter() {group=Background, parameterName=Smoke Color, defaultValue=<rgba 1 0 0.0733 1>}
  n2 = NamedRerouteUsage()
  n3 = Min(
   A=@n2
   B=MaterialFunctionCall(Texture=NamedRerouteUsage(), GradientStart=Constant() {r=-0.1}, GradientEnd=Constant() {r=1.1}, RotationAngle=Constant() {r=0.75}) {materialFunction=&MF_UI_LinearGradient}.Result
  )
  n4 = LinearInterpolate(A=NamedRerouteUsage(), B=Multiply(A=Multiply(A=@n1.RGB, B=@n3)) {constB=0.6}, Alpha=Max(A=@n2, B=NamedRerouteUsage()))
  n5 = Multiply(A=NamedRerouteUsage()) {constB=1.75}
  MaterialFunctionCall!(
   Color=LinearInterpolate(A=Constant3Vector() {constant=<rgba 0.0059 0.008 0.0104 1>}, B=Add(A=@n4, B=@n5), Alpha=NamedRerouteUsage())
   Color Brightness
   Floor Radius
   Floor Density=Constant() {r=2}
   Floor Metallic
   Floor Specular
   Floor Roughness
  ) {materialFunction=/Game/UI/Foundation/Environment/Shell/Materials/MaterialFunctions/MF_UI_MPItemShop_Setup}
  n6 = NamedRerouteUsage()
  n7 = Multiply(A=@n6)
  n8 = Constant() {r=0.5}
  n9 = Constant() {r=2}
  n10 = Panner(Coordinate=@n6) {speedX=-0.002, speedY=-0.004}
  n11 = SmoothStep(Value=TextureSample(UVs=Multiply(A=@n10) {constB=5}) {samplerType=SAMPLERTYPE_Normal, texture=&/T_UI_Particles}.R) {constMax=0.5}
  Multiply(A=Min(
    A=MaterialFunctionCall(UV=@n7, GradientStart=@n8, GradientEnd=@n9) {materialFunction=&MF_UI_RadialGradient}.Smooth
    B=@n11
   ))
  n12 = NamedRerouteUsage()
  n13 = NamedRerouteUsage()
  NamedRerouteDeclaration(None=Multiply(A=Min(A=@n12, B=@n13), B=Constant3Vector() {constant=<rgba 1 0.0677 0.0677 1>})) {name=Thin Laser Colored, nodeColor=<rgba 0 0.8828 1 1>}
  n14 = NamedRerouteUsage()
  n15 = Min(A=Min(A=NamedRerouteUsage(), B=NamedRerouteUsage()), B=SmoothStep(Value=@n14) {constMax=2, constMin=0.75})
  n16 = LinearInterpolate(B=Max(A=Multiply(A=@n15) {constB=2}, B=NamedRerouteUsage()), Alpha=@n14)
  NamedRerouteDeclaration(None=LinearInterpolate(A=Constant3Vector() {constant=<rgba 0.0091 0.0123 0.016 1>}, B=Constant3Vector() {constant=<rgba 1 0 0 0>}, Alpha=@n16)) {name=SmokeLasers Combined, nodeColor=<rgba 0.7109 1 0 1>}
  n17 = ScalarParameter() {group=Lasers, parameterName=Laser X Location 2}
  n18 = ScalarParameter() {defaultValue=0.2, group=Lasers, parameterName=Laser Y Location 2}
  n19 = Constant()
  n20 = MaterialFunctionCall(
   UVs=Add(A=NamedRerouteUsage(), B=AppendVector(A=@n17, B=@n18))
   "Rotation Angle (0-1)"=MaterialFunctionCall(
    Input=ScalarParameter() {defaultValue=30, group=Lasers, parameterName=Angle of Laser 2}
    Input Low=@n19
    Input High=Constant() {r=-360}
    Target Low=@n19
    Target High=Constant() {r=1}
   ) {materialFunction=&RemapValueRange}.Result
  ) {materialFunction=&CustomRotator}
  NamedRerouteDeclaration(None=@n20.Rotated Values) {name=Laser 2 UVs, nodeColor=<rgba 0.1016 0 1 1>}
  n21 = MaterialFunctionCall(UVs=NamedRerouteUsage(), "Rotation Angle (0-1)"=Constant() {r=0.75}) {materialFunction=&CustomRotator}
  n22 = Constant() {r=0.75}
  n23 = Constant() {r=1}
  n24 = Constant() {r=1}
  NamedRerouteDeclaration(None=OneMinus(None=MaterialFunctionCall!(UVs=@n21.Rotated Values, Subtract=@n22, Falloff=@n23, Intensity=@n24) {materialFunction=/Game/UI/Frontend/ItemShop/Art/_MaterialFunctions_ShopTheming/MF_UI_TopBottomGradient}.Result)) {name=SideGradients, nodeColor=<rgba 0.5 1 0 1>}
  n25 = NamedRerouteUsage()
  n26 = MaterialFunctionCall!(
   Texture 2D=TextureObject() {samplerType=SAMPLERTYPE_Grayscale, texture=/Game/UI/Frontend/ItemShop/Art/Rumble/LowResBlurredNoise_Rumble}
   Texture UVs=Panner(Coordinate=Multiply(A=@n25) {constB=0.8942}) {speedY=0.05}
   FlowMap Texture 2D=TextureObject() {samplerType=SAMPLERTYPE_LinearColor, texture=/Game/UI/Frontend/ItemShop/Art/WaterBackground/T_UI_WaterFlowMap}
   FlowMap UVs=NamedRerouteUsage()
   FlowIntensity=Constant() {r=0.7583}
   FlowSpeed=Constant()
  ) {materialFunction=/FortUILibrary/MaterialFunctions/Effects/MF_UI_FlowMap}
  n27 = Multiply(A=@n25) {constB=2.5}
  n28 = Multiply(A=Time(), B=ScalarParameter() {defaultValue=0.3, parameterName=Time Factor})
  n29 = TextureSample(UVs=Panner(Coordinate=@n27, Time=@n28) {speedX=0.05, speedY=0.1}) {samplerSource=SSM_Wrap_WorldGroupSettings, texture=/Game/UI/Frontend/ItemShop/Art/IWDTakeover/Texture/T-Noise-IWD}
  n30 = TextureSample(UVs=Panner(Coordinate=@n27, Time=@n28) {speedX=0.1, speedY=0.012}) {samplerType=SAMPLERTYPE_Grayscale, texture=&LowResBlurredNoise}
  n31 = LinearInterpolate(A=@n26.Result, Alpha=Multiply(A=Add(A=@n29.R, B=@n30.R)) {constB=1.4})
  NamedRerouteDeclaration(None=@n31) {name=Cloud Mask no Gradient, nodeColor=<rgba 0 1 0.9062 1>}
  n32 = ScalarParameter() {group=Lasers, parameterName=Laser X Location 4}
  n33 = ScalarParameter() {defaultValue=-0.1, group=Lasers, parameterName=Laser Y Location 4}
  n34 = Constant()
  n35 = MaterialFunctionCall(
   UVs=Add(A=NamedRerouteUsage(), B=AppendVector(A=@n32, B=@n33))
   "Rotation Angle (0-1)"=MaterialFunctionCall(
    Input=ScalarParameter() {defaultValue=25, group=Lasers, parameterName=Angle of Laser 4}
    Input Low=@n34
    Input High=Constant() {r=-360}
    Target Low=@n34
    Target High=Constant() {r=1}
   ) {materialFunction=&RemapValueRange}.Result
  ) {materialFunction=&CustomRotator}
  n36 = Constant() {r=0.005}
  n37 = Constant() {r=2}
  n38 = OneMinus(None=MaterialFunctionCall(UVs=@n35.Rotated Values, Size=AppendVector(A=@n36, B=@n37), CornerRoundness=Constant(), GlowMin=Constant(), GlowMax=Constant() {r=0.05}) {materialFunction=&MF_UI_Box}.SDF)
  n39 = Add(
   A=NamedRerouteUsage()
   B=AppendVector(
    A=ScalarParameter() {group=Lasers, parameterName=Laser X Location 3}
    B=ScalarParameter() {defaultValue=0.1, group=Lasers, parameterName=Laser Y Location 3}
   )
  )
  n40 = Constant()
  n41 = MaterialFunctionCall(
   Input=ScalarParameter() {defaultValue=0.3, group=Lasers, parameterName=Angle of Laser 3}
   Input Low=@n40
   Input High=Constant() {r=-360}
   Target Low=@n40
   Target High=Constant() {r=1}
  ) {materialFunction=&RemapValueRange}
  n42 = Constant() {r=0.005}
  n43 = Constant() {r=2}
  n44 = OneMinus(None=MaterialFunctionCall(
    UVs=MaterialFunctionCall(UVs=@n39, "Rotation Angle (0-1)"=@n41.Result) {materialFunction=&CustomRotator}.Rotated Values
    Size=AppendVector(A=@n42, B=@n43)
    CornerRoundness=Constant()
    GlowMin=Constant()
    GlowMax=Constant() {r=0.05}
   ) {materialFunction=&MF_UI_Box}.SDF)
  n45 = Max(A=SmoothStep(Value=@n38) {constMax=1.01, constMin=1}, B=SmoothStep(Value=@n44) {constMax=1.01, constMin=1})
  n46 = Constant() {r=0.005}
  n47 = Constant() {r=2}
  n48 = OneMinus(None=MaterialFunctionCall(UVs=@n20.Rotated Values, Size=AppendVector(A=@n46, B=@n47), CornerRoundness=Constant(), GlowMin=Constant(), GlowMax=Constant() {r=0.05}) {materialFunction=&MF_UI_Box}.SDF)
  n49 = SmoothStep(Value=@n48) {constMax=1.01, constMin=0.997}
  n50 = ScalarParameter() {group=Lasers, parameterName=Laser X Location}
  n51 = ScalarParameter() {group=Lasers, parameterName=Laser Y Location}
  n52 = Constant()
  n53 = MaterialFunctionCall(
   UVs=Add(A=NamedRerouteUsage(), B=AppendVector(A=@n50, B=@n51))
   "Rotation Angle (0-1)"=MaterialFunctionCall(
    Input=ScalarParameter() {defaultValue=25, group=Lasers, parameterName=Angle of Laser}
    Input Low=@n52
    Input High=Constant() {r=-360}
    Target Low=@n52
    Target High=Constant() {r=1}
   ) {materialFunction=&RemapValueRange}.Result
  ) {materialFunction=&CustomRotator}
  n54 = Constant() {r=0.005}
  n55 = Constant() {r=2}
  n56 = OneMinus(None=MaterialFunctionCall(UVs=@n53.Rotated Values, Size=AppendVector(A=@n54, B=@n55), CornerRoundness=Constant(), GlowMin=Constant(), GlowMax=Constant() {r=0.05}) {materialFunction=&MF_UI_Box}.SDF)
  NamedRerouteDeclaration(None=Max(A=Max(A=@n45, B=@n49), B=SmoothStep(Value=@n56) {constMax=1.01, constMin=0.997})) {name=Thin Laser Mask, nodeColor=<rgba 0 0.3906 1 1>}
  n57 = MaterialFunctionCall!(ScreenUV, TextureSize, ScalePivotPoint) {materialFunction=/FortUILibrary/MaterialFunctions/Basics/MF_UI_ScreenSpaceUVScaling}
  NamedRerouteDeclaration(None=@n57.Scaled to Fill) {name=UV Scaling Fill, nodeColor=<rgba 0 0.9766 1 1>}
  NamedRerouteDeclaration(None=Max(A=SmoothStep(Value=@n38) {constMax=1.5, constMin=0.75}, B=SmoothStep(Value=@n48) {constMin=0.75})) {name=Large Haze Mask, nodeColor=<rgba 1 0 0.4688 1>}
  NamedRerouteDeclaration(None=@n16) {name=SmokeLasers Mask, nodeColor=<rgba 0 0.6953 1 1>}
  n58 = Max(
   A=Multiply(A=SmoothStep(Value=@n38) {constMax=1.1, constMin=0.95}) {constB=0.3}
   B=Multiply(A=SmoothStep(Value=@n44) {constMax=1.1, constMin=0.95}) {constB=0.4}
  )
  n59 = Multiply(A=SmoothStep(Value=@n48) {constMax=1.1, constMin=0.95}) {constB=0.5}
  NamedRerouteDeclaration(None=Max(A=Max(A=@n58, B=@n59), B=SmoothStep(Value=@n56) {constMax=1.1, constMin=0.95})) {name=Large Laser Mask, nodeColor=<rgba 0.3594 0 1 1>}
  n60 = Constant() {r=0.088}
  n61 = Constant() {r=1.0594}
  n62 = Constant() {r=0.75}
  NamedRerouteDeclaration(None=LinearInterpolate(
    B=@n31
    Alpha=MaterialFunctionCall(Texture=@n26.Result, GradientStart=@n60, GradientEnd=@n61, RotationAngle=@n62) {materialFunction=&MF_UI_LinearGradient}.Result
   ) {constA=0.3}) {name=Cloud Mask, nodeColor=<rgba 0 1 0.2969 1>}
  NamedRerouteDeclaration(None=@n53.Rotated Values) {name=Lasers 1 UVs, nodeColor=<rgba 1 0.2578 0 1>}
  NamedRerouteDeclaration(None=@n57.Scaled to Fit) {name=UVScaling, nodeColor=<rgba 0.2188 1 0 1>}
  NamedRerouteDeclaration(None=@n35.Rotated Values) {name=Laser4 UVs, nodeColor=<rgba 1 0 0.7385 1>}
  NamedRerouteDeclaration(None=@n11) {name=Particles, nodeColor=<rgba 0.4062 1 0 1>}
  NamedRerouteUsage()
 ]
}
